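# Masked, lit cardboard material for mesh particles in the StarGazer guitar emote, with randomized UVs per particle.
# An optional construction paper layer adds a particle-colored center cutout with an offset shadow and flattened normals, scaled, rotated and offset through dynamic parameters.
# Two flipbook-picked scribbles from the packed texture overlay it, tinted by hue-shifted, saturation-adjusted variants of particle color derived through HSV.
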Material /SparksCosmetics/Emotes/Guitar/StarGazer/Effects/Materials/M_Emote_StarGazer_Cardboard
graph {
 properties {bAutomaticallySetUsageInEditor=false, bUsedWithNiagaraMeshParticles=true, blendMode=BLEND_Masked, usedShadingModels=Default Lit}
 nodes [
  n1 = Add(A=ParticleRandom(), B=Multiply(A=TextureCoordinate() {coordinateIndex=1}) {constB=0.25})
  n2 = TextureSample(UVs=@n1) {texture=/BRCosmetics/Weapons/WeaponSkins/Wraps/Textures/CardboardCrew/T_Wrap_CardboardCrew_BC}
  n3 = DynamicParameter() {defaultValue=<rgba 0.7 0.5 0.04 0>, paramNames [Scale Center Star, Rotate Center Star, Offset Center Star U, Offset Center Star V]}
  n4 = Add(
   A=Rotator(
    Coordinate=MaterialFunctionCall(Texture Scale=@n3.Scale Center Star) {materialFunction=&ScaleUVsByCenter}.UVs
    Time=@n3.Rotate Center Star
   )
   B=AppendVector(A=@n3.Offset Center Star U, B=@n3.Offset Center Star U)
  )
  n5 = Saturate(None=Add(A=@n4) {constB=-0.02})
  n6 = OneMinus(None=ChannelMaskParameter(None=TextureSample(UVs=@n5) {samplerType=SAMPLERTYPE_LinearColor, texture=/SparksCosmetics/Emotes/Guitar/StarGazer/Effects/Textures/T_Emote_StarGazer_Pack}.RGBA) {maskChannel=Green, parameterName=Shape Mask, defaultValue=<rgba 0 1 0 0>}.R)
  n7 = ChannelMaskParameter(None=TextureSample() {samplerType=SAMPLERTYPE_LinearColor, texture=&T_Emote_StarGazer_Pack}.RGBA) {maskChannel=Green, parameterName=Shape Mask, defaultValue=<rgba 0 1 0 0>}
  n8 = LinearInterpolate(
   A=LinearInterpolate(A=Multiply(A=@n2.RGB) {constB=0.25}, B=@n2.RGB, Alpha=Multiply(A=@n6, B=@n7.R))
   B=ParticleColor().RGB
   Alpha=NamedRerouteUsage()
  )
  n9 = NamedRerouteUsage()
  n10 = NamedRerouteUsage()
  n11 = DynamicParameter() {
   parameterIndex=1
   defaultValue=<rgba 0 0.5 1 1.75>
   paramNames ["Pick Scribble A (0-1)", "Pick Scribble B (0-1)", Use Scribble, Scribble Saturation Mult]
  }
  n12 = Constant() {r=2}
  n13 = TextureObject() {samplerType=SAMPLERTYPE_LinearColor, texture=&T_Emote_StarGazer_Pack}
  n14 = MaterialFunctionCall(Float3=MaterialFunctionCall("Animation  Phase (0-1)"=@n11."Pick Scribble A (0-1)", Number of Rows=@n12, Number of Columns=@n12, Texture=@n13) {materialFunction=&FlipBook}.Result) {materialFunction=&BreakOutFloat3Components}
  n15 = LinearInterpolate(A=@n9, B=@n10, Alpha=@n14.B)
  n16 = MaterialFunctionCall(Float3=MaterialFunctionCall("Animation  Phase (0-1)"=@n11."Pick Scribble B (0-1)", Number of Rows=@n12, Number of Columns=@n12, Texture=@n13) {materialFunction=&FlipBook}.Result) {materialFunction=&BreakOutFloat3Components}
  n17 = LinearInterpolate(A=@n10, B=@n9, Alpha=@n16.B)
  n18 = NamedRerouteUsage()
  n19 = TextureSample() {samplerType=SAMPLERTYPE_LinearColor, texture=&T_Emote_StarGazer_Pack}
  n20 = DynamicParameter() {
   parameterIndex=1
   defaultValue=<rgba 0 0.5 1 1.75>
   paramNames ["Pick Scribble A (0-1)", "Pick Scribble B (0-1)", Use Scribble, Scribble Saturation Mult]
  }
  n21 = LinearInterpolate(B=Saturate(None=Add(A=@n14.B, B=@n16.B)), Alpha=@n20.Use Scribble)
  n22 = Multiply(
   A=Multiply(
    A=LinearInterpolate(Alpha=@n18) {constA=0.5}
    B=ChannelMaskParameter(None=@n19.RGBA) {maskChannel=Green, parameterName=Shape Mask, defaultValue=<rgba 0 1 0 0>}.R
   )
   B=@n21
  )
  n23 = Constant4Vector() {constant=<rgba 0 0 0.25 0.25>}
  Multiply(
   A=LinearInterpolate(
    A=StaticSwitchParameter(True=@n8, False=@n2.RGB) {defaultValue=true, parameterName=Use Construction Paper}
    B=Multiply(A=@n15, B=@n17)
    Alpha=StaticSwitchParameter(True=@n22, False=@n21) {defaultValue=true, parameterName=Use Construction Paper}
   )
   B=MaterialFunctionCall(Value_Morning=@n23.R, Value_Day=@n23.R, Value_Evening=@n23.R, Value_Night=@n23.R) {materialFunction=&Scalar_Time_of_Day}.Result
  ) {constB=0.25}
  n24 = Saturate(None=@n4)
  NamedRerouteDeclaration(None=ChannelMaskParameter(None=TextureSample(UVs=@n24) {samplerType=SAMPLERTYPE_LinearColor, texture=&T_Emote_StarGazer_Pack}.RGBA) {maskChannel=Green, parameterName=Shape Mask, defaultValue=<rgba 0 1 0 0>}.R) {name=ConstructionPaperCutout, nodeColor=<rgba 1 0 0.9155 1>}
  n25 = MaterialFunctionCall(Float3=MaterialFunctionCall!(In=ParticleColor().RGB) {materialFunction=/Engine/Functions/Engine_MaterialFunctions01/Chromakeying/RGBtoHSV}.Result) {materialFunction=&BreakOutFloat3Components}
  NamedRerouteDeclaration(None=@n25.G) {name=Saturation, nodeColor=<rgba 0.2891 0 1 1>}
  NamedRerouteDeclaration(None=@n25.B) {name=Value, nodeColor=<rgba 0.2891 0 1 1>}
  n26 = Fmod(A=Add(A=@n25.R, B=Constant() {r=-0.1111}), B=Constant() {r=1})
  n27 = If(A=@n26, "A > B"=@n26, "A == B"=@n26, "A < B"=Add(A=@n26))
  n28 = DynamicParameter() {
   parameterIndex=1
   defaultValue=<rgba 0 0.5 1 1.75>
   paramNames ["Pick Scribble A (0-1)", "Pick Scribble B (0-1)", Use Scribble, Scribble Saturation Mult]
  }
  n29 = Subtract(A=Clamp(None=@n28.Scribble Saturation Mult) {maxDefault=2})
  n30 = NamedRerouteUsage()
  n31 = OneMinus(None=@n30)
  n32 = If(
   A=@n29
   "A > B"=Add(A=@n30, B=Multiply(A=@n31, B=@n29))
   "A == B"=NamedRerouteUsage()
   "A < B"=Multiply(A=NamedRerouteUsage(), B=Add(A=@n29))
  )
  n33 = NamedRerouteUsage()
  NamedRerouteDeclaration(None=HsvToRgb!(None=MaterialFunctionCall(X=@n27, Y=@n32, Z=@n33) {materialFunction=&MakeFloat3}.Result)) {name=ModifiedColorA, nodeColor=<rgba 0.5938 0 1 1>}
  n34 = Fmod(A=Add(A=Constant() {r=0.0833}, B=@n25.R), B=Constant() {r=1})
  n35 = If(A=@n34, "A > B"=@n34, "A == B"=@n34, "A < B"=Add(A=@n34))
  n36 = NamedRerouteUsage()
  NamedRerouteDeclaration(None=HsvToRgb(None=MaterialFunctionCall(X=@n35, Y=@n32, Z=@n36) {materialFunction=&MakeFloat3}.Result)) {name=ModifiedColorB, nodeColor=<rgba 0.5938 0 1 1>}
  n37 = TextureSample(UVs=@n1) {samplerType=SAMPLERTYPE_Normal, texture=/BRCosmetics/Weapons/WeaponSkins/Wraps/Textures/CardboardCrew/T_Wrap_CardboardCrew_N}
  n38 = Constant3Vector() {constant=<rgba 0 0 1 0>}
  n39 = LinearInterpolate(A=@n37.RGB, B=@n38, Alpha=NamedRerouteUsage())
  Reroute(None=LinearInterpolate(
    A=StaticSwitchParameter(True=@n39, False=@n37.RGB) {defaultValue=true, parameterName=Use Construction Paper}
    B=@n38
    Alpha=ScalarParameter() {defaultValue=0.5, parameterName=Reduce Normal}
   ))
  StaticSwitchParameter(True=LinearInterpolate(Alpha=NamedRerouteUsage()) {constA=0.5, constB=0.9}, False=Constant() {r=0.5}) {defaultValue=true, parameterName=Use Construction Paper}
 ]
}
