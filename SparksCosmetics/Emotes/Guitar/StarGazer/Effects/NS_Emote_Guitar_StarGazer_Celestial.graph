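# Niagara system for the StarGazer guitar emote that builds a celestial backdrop of drifting clouds, a moon, stars and a shooting star.
# Clouds and the shooting star follow position curves with a sine-like bob and scale in and out.
# Staggered star bursts are distributed across box regions, then share system-level bounce curves with the moon for motion and spin, plus random drift.
# Several disabled Minimal emitters visualize spawn boxes in flat colors.
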NiagaraSystem /SparksCosmetics/Emotes/Guitar/StarGazer/Effects/NS_Emote_Guitar_StarGazer_Celestial
dependencies {
 usedDataInterfaces [
  &NiagaraDataInterfaceArrayColor
  &NiagaraDataInterfaceCurve
  /Script/Niagara.NiagaraDataInterfaceArrayFloat4
  &NiagaraDataInterfaceVectorCurve
  &NiagaraDataInterfaceColorCurve
  &NiagaraDataInterfaceArrayFloat3
  /Script/Niagara.NiagaraDataInterfaceSimpleCounter
 ]
 usedDynamicInputs []
 usedModules [
  &SystemState
  &self:SystemSpawnScript.SystemScriptSource.SystemScriptGraph.NiagaraNodeAssignment_0.SetVariables_#1
 ]
}
emitters {
 Clouds {
  emitterUpdateScript [
   &EmitterState()
   &SpawnRate(
    SpawnRate=&RandomRangeFloat(Minimum=0.2, Maximum=0.5, Evaluation Type=::NewEnumerator1)
    Use Spawn Probability=false
   )
  ]
  particleSpawnScript [
   /SetVariables(Particles.Variance:NiagaraFloat=&RandomRangeFloat(Minimum=0, Maximum=1)) {enabled=false, isSetParameters=true}
   /SetVariables(Particles.Variance:NiagaraFloat=&FloatFromCurve(
     FloatCurve=DataInterface() {propertyValues {Curve {keys [
         {arriveTangent=0.2936, interpMode=RCIM_Cubic, leaveTangent=0.2936, tangentMode=RCTM_User, time=0, value=0}
         {arriveTangent=0.2542, interpMode=RCIM_Cubic, leaveTangent=0.2542, tangentMode=RCTM_User, time=1, value=1}
        ]}}}
     CurveIndex=&RandomRangeFloat(Minimum=0, Maximum=1)
    )) {isSetParameters=true}
   &InitializeParticle(
    Color=&ColorFromCurve(
     ColorCurve=DataInterface() {propertyValues {
       AlphaCurve {keys [<at 0 1 RCIM_Cubic>]}
       BlueCurve {keys [<at 0 1 RCIM_Cubic>,<at 0.4821 0.7858>,<at 1 0.9725 RCIM_Cubic>]}
       GreenCurve {keys [<at 0 0.7461 RCIM_Cubic>,<at 0.4821 0.574>,<at 1 1 RCIM_Cubic>]}
       RedCurve {keys [<at 0 0.9365 RCIM_Cubic>,<at 0.4821 0.5316>,<at 1 0.6696 RCIM_Cubic>]}
      }}
     CurveIndex=&RandomRangeFloat(Minimum=0, Maximum=1)
    )
    Lifetime=&Lerp_Float(A=4, B=10, Alpha=$Particles.Variance)
    Mesh Scale=&self:ShareYandZ(X Scale:NiagaraFloat=1, Y and Z Scale:NiagaraFloat=&Lerp_Float(A=0.8, B=1.2, Alpha=$Particles.Variance))
    Mesh Scale Mode=::NewEnumerator4
    UsePositionOffset=false
    Write Material Random=true
   )
   &InitialMeshOrientation(
    Mesh Axis To Orient=&SelectVectorFromArray(
     Array Sampling Mode=::NewEnumerator0
     Direct Array Index=?
     Vector Selection Array=DataInterface() {propertyValues {FloatData [<0 -1 0>, <0 1 0>]}}
     Random Seed=0
    )
    Mesh Orientation Mode=::NewEnumerator2
    Use Rotation Vector=false
   )
   /SetVariables(Particles.MeshIndex=&ModuloInt(
     A=/Niagara/DynamicInputs/DataInterface/GenerateIntFromCounter(SimpleCounter:NiagaraDataInterfaceSimpleCounter=DataInterface() {propertyValues {GpuSyncMode=None, InitialValue=0}})
     B=3
    )) {isSetParameters=true}
   &DynamicMaterialParameters(
    Index 0 Vector 4:Vector4f=/Niagara/DynamicInputs/Arrays/SelectVector4_FromArray(
     Array Sampling Mode:ENiagaraArraySamplingMode=::NewEnumerator0
     Direct Array Index:NiagaraInt32=?
     Interpolation:NiagaraFloat=?
     Direct Array Mode:ENiagara_WrapClamp=::NewEnumerator0
     Evaluation Type:ENiagaraRandomnessEvaluation=::NewEnumerator0
     Normalized Interpolation:NiagaraBool=true
     Recalculate Random Each Loop:NiagaraBool=false
     Randomness Mode:ENiagaraRandomnessMode=::NewEnumerator0
     Random Seed:NiagaraInt32=0
     Fixed Random Seed:NiagaraBool=false
     Vector4 Selection Array:NiagaraDataInterfaceArrayFloat4=DataInterface() {propertyValues {GpuSyncMode=SyncCpuToGpu, MaxElements=0, FloatData [<0.82 0 0.04 0.02>, <0.73 0 -0.03 -0.06>, <0.8 0 0 0>, <0.9 0 0.01 -0.01>]}}
    )
    Index 1 Param 1=&RandomRangeFloat(Minimum=0, Maximum=1)
    Index 1 Param 2=&RandomRangeFloat(Minimum=0, Maximum=1)
    Index 1 Param 3=&RandomRangeFloat(Minimum=0.5, Maximum=0.75)
    Index 1 Param 4=1.2
    Module.Index1Float0Write=true
    Module.Index1Float1Write=true
    Module.Index1Float2Write=true
    Module.Index1Float3Write=true
    Write Parameter Index 0=::NewEnumerator4
    Write Parameter Index 1=::NewEnumerator1
   )
  ]
  particleUpdateScript [
   &ParticleState()
   &ScaleMeshSize(Scale Factor=&VectorFromFloat(Value=&FloatFromCurve(FloatCurve=DataInterface() {propertyValues {Curve {keys [
          {arriveTangent=17.1544, interpMode=RCIM_Cubic, leaveTangent=17.1544, tangentMode=RCTM_User, time=0, value=0}
          {interpMode=RCIM_Cubic, tangentMode=RCTM_User, time=0.05, value=1}
          {arriveTangentWeight=0.2325, interpMode=RCIM_Cubic, leaveTangentWeight=0.1744, tangentMode=RCTM_User, time=0.95, value=1}
          {arriveTangent=-24.4206, tangentMode=RCTM_User, time=1, value=0}
         ]}}})))
   /SetVariables(Transient.SpinPhase=0.2) {isSetParameters=true}
   /SetVariables(Transient.CenterSpin:NiagaraFloat=&Divide_Float(A=0.25, B=$Transient.SpinPhase)) {isSetParameters=true}
   /SetVariables(Particles.Bob:NiagaraFloat=&Waveform(
     Global Phase=&Add_Float(A=$Particles.Age, B=$Transient.CenterSpin)
     "Waveform [1]"=::NewEnumerator0
     "[1] Phase Scale"=$Transient.SpinPhase
     "[1] Frequency"=1
     "[1] Amplitude Min/Max"=<-1 1>
    )) {isSetParameters=true}
   /SetVariables(Particles.Position=&Add_Vector(
     A=&VectorFromCurve(VectorCurve=DataInterface() {propertyValues {
        XCurve {keys [
          {arriveTangent=118.14, interpMode=RCIM_Cubic, leaveTangent=118.14, leaveTangentWeight=153.2665, tangentMode=RCTM_User, time=0, value=-54}
          {arriveTangent=126.5729, arriveTangentWeight=170.1042, interpMode=RCIM_Cubic, leaveTangent=126.5729, tangentMode=RCTM_User, time=1, value=66}
         ]}
        YCurve {keys [<at 0 -40>]}
        ZCurve {keys [{arriveTangent=59.6951, interpMode=RCIM_Cubic, leaveTangent=59.6951, tangentMode=RCTM_User, time=0, value=-50}]}
       }})
     B=&RandomVector(Vector Scale=20)
    )) {enabled=false, isSetParameters=true, module=SetVariables_#2}
   /SetVariables(Particles.Position=&MakeVector(
     X=&FloatFromCurve(FloatCurve=DataInterface() {propertyValues {Curve {keys [
          {arriveTangent=118.14, interpMode=RCIM_Cubic, leaveTangent=118.14, leaveTangentWeight=0.3333, tangentMode=RCTM_User, time=0, value=-54}
          {arriveTangent=126.5729, arriveTangentWeight=0.3333, interpMode=RCIM_Cubic, leaveTangent=126.5729, tangentMode=RCTM_User, time=1, value=66}
         ]}}})
     Y=&Lerp_Float(A=-7, B=-20, Alpha=$Particles.Variance)
     Z=&Add_Float(
      A=&RandomRangeFloat(Minimum=-50, Maximum=-30)
      B=&Multiply_Float(A=-1, B=$Particles.Bob)
     )
    )) {isSetParameters=true}
   /SetVariables(Particles.Position=&MakeVector(
     X=&FloatFromCurve(FloatCurve=DataInterface() {propertyValues {Curve {keys [
          {arriveTangent=118.14, interpMode=RCIM_Cubic, leaveTangent=118.14, leaveTangentWeight=0.3333, tangentMode=RCTM_User, time=0, value=-54}
          {arriveTangent=126.5729, arriveTangentWeight=0.3333, interpMode=RCIM_Cubic, leaveTangent=126.5729, tangentMode=RCTM_User, time=1, value=66}
         ]}}})
     Y=&Lerp_Float(A=30, B=40, Alpha=$Particles.Variance)
     Z=&Add_Float(
      A=&RandomRangeFloat(Minimum=-30, Maximum=-10)
      B=&Multiply_Float(A=-1, B=$Particles.Bob)
     )
    )) {enabled=false, isSetParameters=true}
   &UpdateMeshOrientation(Rotation Rate=&Multiply_Float(A=$Particles.Bob, B=0.015))
  ]
  rendererClasses [&NiagaraMeshRendererProperties]
  renderers [{rendererClass=&NiagaraMeshRendererProperties, rendererIndex=0}]
 }
 Minimal {
  bEnabled=false
  emitterUpdateScript [
   &EmitterState()
   &SpawnRate(SpawnRate=50, Use Spawn Probability=false)
  ]
  particleSpawnScript [
   &InitializeParticle(Color Mode=::NewEnumerator0, Lifetime=5, Sprite Size Mode=::NewEnumerator3, UsePositionOffset=false, Write Material Random=true)
   &ShapeLocation(
    Box Size=<50 100 70>
    MaskSpawnGroup=false
    Offset=<13 0 -25>
    Offset Mode=::NewEnumerator0
    Override Seed=false
    Shape Primitive=::NewEnumerator2
    Surface Only=false
    UseOwnerScale=true
   ) {enabled=false}
   &ShapeLocation(
    Box Midpoint=<0.5 1 0.5>
    Box Size=<130 25 130>
    MaskSpawnGroup=false
    Offset Mode=::NewEnumerator0
    Override Seed=false
    Shape Primitive=::NewEnumerator2
    Surface Only=false
    UseOwnerScale=true
   ) {enabled=false, module=ShapeLocation001}
   &ShapeLocation(MaskSpawnGroup=false, Offset=<13 0 -28>, Offset Mode=::NewEnumerator0, Override Seed=false, Sphere Radius=50, UseOwnerScale=true) {enabled=false, module=ShapeLocation002}
   &ShapeLocation(
    Apply To Particle Position=::NewEnumerator1
    Box Size=<20 25 100>
    MaskSpawnGroup=false
    Offset=&Lerp_Vector(
     A=<-60 0 0>
     B=<73 0 0>
     Alpha=&Multiply_FloatByInt(Float=0.1667, Integer=&ModuloInt(A=&ReturnParticleIDAttributes(), B=6))
    )
    Offset Mode=::NewEnumerator0
    Override Seed=false
    Shape Primitive=::NewEnumerator2
    Surface Only=false
    UseOwnerScale=true
   ) {enabled=false, module=ShapeLocation003}
   &ShapeLocation(
    Box Size=<130 40 130>
    MaskSpawnGroup=false
    Offset=<13 -50 -28>
    Offset Mode=::NewEnumerator0
    Override Seed=false
    Shape Primitive=::NewEnumerator2
    Surface Only=false
    UseOwnerScale=true
   ) {module=ShapeLocation004}
   /SetVariables(Particles.Position=&Add_Vector(
     A=<13 -40 -28>
     B=&Add_Vector(
      A=&Multiply_Vector(A=$Particles.ShapeLocation003.ShapeVector, B=<0 1 0>)
      B=&self:ClampVectorByLength(
       Vector:Vector3f=&Multiply_Vector(A=$Particles.ShapeLocation003.ShapeVector, B=<1 0 1>)
       Max Length:NiagaraFloat=180
       Min Length:NiagaraFloat=40
      )
     )
    )) {enabled=false, isSetParameters=true}
   /SetVariables(Particles.Position=&Add_Vector(
     A=<13 -40 -28>
     B=&Add_Vector(
      A=&Multiply_Vector(A=$Particles.ShapeLocation004.ShapeVector, B=<0 1 0>)
      B=&self:ClampVectorByLength(
       Vector:Vector3f=&Multiply_Vector(A=$Particles.ShapeLocation004.ShapeVector, B=<1 0 1>)
       Max Length:NiagaraFloat=180
       Min Length:NiagaraFloat=40
      )
     )
    )) {enabled=false, isSetParameters=true}
  ]
  particleUpdateScript [&ParticleState()]
  rendererClasses [&NiagaraSpriteRendererProperties]
  renderers [{rendererClass=&NiagaraSpriteRendererProperties}]
 }
 Minimal001 {
  bEnabled=false
  emitterUpdateScript [
   &EmitterState()
   &SpawnRate(SpawnRate=50, Use Spawn Probability=false)
  ]
  particleSpawnScript [
   &InitializeParticle(Color=<rgba 1 0 0 1>, Lifetime=5, Sprite Size Mode=::NewEnumerator3, UsePositionOffset=false, Write Material Random=true)
   &ShapeLocation(
    Box Size=<40 100 75>
    MaskSpawnGroup=false
    Offset=<-32 0 -25>
    Offset Mode=::NewEnumerator0
    Override Seed=false
    Shape Origin=$Engine.Emitter.SimulationPosition
    Shape Primitive=::NewEnumerator2
    Surface Only=false
    UseOwnerScale=true
   ) {enabled=false}
   &ShapeLocation(
    Box Size=<60 40 70>
    MaskSpawnGroup=false
    Offset=<40 -50 0>
    Offset Mode=::NewEnumerator0
    Override Seed=false
    Shape Primitive=::NewEnumerator2
    Surface Only=false
    UseOwnerScale=true
   ) {module=ShapeLocation004}
  ]
  particleUpdateScript [&ParticleState()]
  rendererClasses [&NiagaraSpriteRendererProperties]
  renderers [{rendererClass=&NiagaraSpriteRendererProperties}]
 }
 Minimal002 {
  bEnabled=false
  emitterUpdateScript [
   &EmitterState()
   &SpawnRate(SpawnRate=50, Use Spawn Probability=false)
  ]
  particleSpawnScript [
   &InitializeParticle(Color=<rgba 0 0 1 1>, Lifetime=5, Sprite Size Mode=::NewEnumerator3, UsePositionOffset=false, Write Material Random=true)
   &ShapeLocation(
    Box Size=<40 100 75>
    MaskSpawnGroup=false
    Offset=<58 0 -25>
    Offset Mode=::NewEnumerator0
    Override Seed=false
    Shape Origin=$Engine.Emitter.SimulationPosition
    Shape Primitive=::NewEnumerator2
    Surface Only=false
    UseOwnerScale=true
   ) {enabled=false}
   &ShapeLocation(
    Box Size=<30 40 100>
    MaskSpawnGroup=false
    Offset=<-35 -50 -13>
    Offset Mode=::NewEnumerator0
    Override Seed=false
    Shape Primitive=::NewEnumerator2
    Surface Only=false
    UseOwnerScale=true
   ) {module=ShapeLocation004}
  ]
  particleUpdateScript [&ParticleState()]
  rendererClasses [&NiagaraSpriteRendererProperties]
  renderers [{rendererClass=&NiagaraSpriteRendererProperties}]
 }
 Minimal003 {
  bEnabled=false
  emitterUpdateScript [
   &EmitterState()
   &SpawnRate(SpawnRate=50, Use Spawn Probability=false)
  ]
  particleSpawnScript [
   &InitializeParticle(Color=<rgba 0 1 0 1>, Lifetime=5, Sprite Size Mode=::NewEnumerator3, UsePositionOffset=false, Write Material Random=true)
   &ShapeLocation(
    Box Size=<130 100 25>
    MaskSpawnGroup=false
    Offset=<13 0 25>
    Offset Mode=::NewEnumerator0
    Override Seed=false
    Shape Origin=$Engine.Emitter.SimulationPosition
    Shape Primitive=::NewEnumerator2
    Surface Only=false
    UseOwnerScale=true
   ) {enabled=false}
   &ShapeLocation(
    Box Size=<40 40 40>
    MaskSpawnGroup=false
    Offset=<60 -50 -50>
    Offset Mode=::NewEnumerator0
    Override Seed=false
    Shape Primitive=::NewEnumerator2
    Surface Only=false
    UseOwnerScale=true
   ) {module=ShapeLocation004}
  ]
  particleUpdateScript [&ParticleState()]
  rendererClasses [&NiagaraSpriteRendererProperties]
  renderers [{rendererClass=&NiagaraSpriteRendererProperties}]
 }
 Minimal004 {
  bEnabled=false
  emitterUpdateScript [
   &EmitterState()
   &SpawnRate(SpawnRate=50, Use Spawn Probability=false)
  ]
  particleSpawnScript [
   &InitializeParticle(Color=<rgba 1 1 0 0.087>, Lifetime=5, Sprite Size Mode=::NewEnumerator3, UsePositionOffset=false, Write Material Random=true)
   &ShapeLocation(
    Apply To Particle Position=::NewEnumerator1
    Box Size=<30 40 100>
    MaskSpawnGroup=false
    Offset=<-35 -50 -13>
    Offset Mode=::NewEnumerator0
    Override Seed=false
    Shape Primitive=::NewEnumerator2
    Surface Only=false
    UseOwnerScale=true
   )
   &ShapeLocation(
    Apply To Particle Position=::NewEnumerator1
    Box Size=<60 40 70>
    MaskSpawnGroup=false
    Offset=<40 -50 0>
    Offset Mode=::NewEnumerator0
    Override Seed=false
    Shape Primitive=::NewEnumerator2
    Surface Only=false
    UseOwnerScale=true
   ) {module=ShapeLocation001}
   &ShapeLocation(
    Apply To Particle Position=::NewEnumerator1
    Box Size=<40 40 40>
    MaskSpawnGroup=false
    Offset=<60 -50 -50>
    Offset Mode=::NewEnumerator0
    Override Seed=false
    Shape Primitive=::NewEnumerator2
    Surface Only=false
    UseOwnerScale=true
   ) {module=ShapeLocation002}
   &self:DistributeLocations(Placement Region:NiagaraInt32=&RandomRangeInt(Minimum=0, Maximum=2)) {module=DistributeLocations}
  ]
  particleUpdateScript [&ParticleState()]
  rendererClasses [&NiagaraSpriteRendererProperties]
  renderers [{rendererClass=&NiagaraSpriteRendererProperties}]
 }
 Moon {
  emitterUpdateScript [
   &EmitterState()
   &SpawnBurst_Instantaneous(Use Spawn Probability=false, UseLoopCountLimit=true)
  ]
  particleSpawnScript [
   &InitializeParticle(
    Color=&SelectLinearColorFromArray(
     Array Sampling Mode=::NewEnumerator1
     Direct Array Index=3
     Color Selection Array=DataInterface() {propertyValues {ColorData [<rgba 0.4434 0.8229 1 1>, <rgba 0.7809 1 0.8123 1>, <rgba 0.9886 1 0.6573 1>, <rgba 0.9809 1 0.4261 1>]}}
     Random Seed=?
    )
    Lifetime=5
    Mesh Scale Mode=::NewEnumerator3
    Sprite Size Mode=::NewEnumerator3
    Uniform Sprite Size=100
    UsePositionOffset=false
    Write Material Random=true
   )
   &InitialMeshOrientation(Mesh Orientation Mode=::NewEnumerator2, Orientation Vector=<0 1 0>, Use Rotation Vector=false)
   &DynamicMaterialParameters(
    Index 0 Param 1=0.8
    Index 0 Param 2=0.1
    Index 0 Param 3=0.06
    Index 0 Param 4=-0.05
    Index 1 Param 1=0.25
    Index 1 Param 2=0.5
    Index 1 Param 3=0.5
    Index 1 Param 4=1.5
    Module.Index0Float0Write=true
    Module.Index0Float1Write=true
    Module.Index0Float2Write=true
    Module.Index0Float3Write=true
    Module.Index1Float0Write=true
    Module.Index1Float1Write=true
    Module.Index1Float2Write=true
    Module.Index1Float3Write=true
    Write Parameter Index 1=::NewEnumerator1
   )
  ]
  particleUpdateScript [
   &ParticleState(Kill Particles When Lifetime Has Elapsed=false, Let Infinitely Lived Particles Die When Emitter Deactivates=true)
   /SetVariables(Transient.AgeAdjusted:NiagaraFloat=&Multiply_Float(A=$Particles.Age, B=2)) {isSetParameters=true}
   /SetVariables(
    Particles.BouncePhase:NiagaraFloat=&FloatFromCurve(FloatCurve=$System.BouncePhase, CurveIndex=$Transient.AgeAdjusted)
    Particles.BouncePosition:NiagaraPosition=&MakeVector(
     X=0
     Y=0
     Z=&Add_Float(
      A=&FloatFromCurve(FloatCurve=$System.BouncePosition, CurveIndex=$Transient.AgeAdjusted, Scale Curve=0.5)
      B=&FloatFromCurve(
       FloatCurve=DataInterface() {propertyValues {Curve {keys [{arriveTangent=-1.9214, interpMode=RCIM_Cubic, leaveTangent=-1.9214, tangentMode=RCTM_User, time=0, value=1}, <at 1 0>]}}}
       CurveIndex=$Transient.AgeAdjusted
       Scale Curve=50
      )
     )
    )
   ) {isSetParameters=true, module=SetVariables_#3}
   &UpdateMeshOrientation(
    Rotation Rate=&Waveform(
     "Waveform [1]"=::NewEnumerator0
     "[1] Phase Scale"=0.1
     "[1] Frequency"=1
     "[1] Amplitude Min/Max"=<-1 1>
     "[1] Scale"=/Niagara/DynamicInputs/Math/LerpMultipleFloats(
      Float Count:ENiagara_MultipleLerpCount=::NewEnumerator1
      1:NiagaraFloat=-5
      2:NiagaraFloat=1
      3:NiagaraFloat=0.1
      4:NiagaraFloat=0.03
      5:NiagaraFloat=?
      Alpha:NiagaraFloat=&Multiply_Float(A=$Particles.BouncePhase, B=0.25)
     )
    )
    Rotation Vector=<0 0 1>
   )
   /SetVariables(
    Particles.GlowColor=&HueShiftLinearColor(Hue=0, Saturation=0.5, Value=0.5, Linear Color=$Particles.Color)
    Particles.GlowOffset=-10
    Particles.GlowPosition:NiagaraPosition=&AddVectorToPosition(Position=$Particles.BouncePosition, Vector=<5 0 -30>)
   ) {enabled=false, isSetParameters=true}
  ]
  rendererClasses [&NiagaraSpriteRendererProperties, &NiagaraMeshRendererProperties]
  renderers [
   {rendererClass=&NiagaraSpriteRendererProperties}
   {rendererClass=&NiagaraMeshRendererProperties, rendererIndex=1}
   {rendererClass=&NiagaraSpriteRendererProperties, rendererIndex=2}
  ]
 }
 ShootingStar {
  emitterUpdateScript [
   &EmitterState()
   &SpawnBurst_Instantaneous(Spawn Time=3, Use Spawn Probability=false, UseLoopCountLimit=false)
  ]
  particleSpawnScript [
   &InitializeParticle(
    Color=<rgba 1 0.7157 0.0922 1>
    Lifetime=5
    Mesh Scale Mode=::NewEnumerator3
    Mesh Uniform Scale=0.8
    Sprite Size Mode=::NewEnumerator3
    Uniform Sprite Size=150
    UsePositionOffset=false
    Write Material Random=true
   )
   &InitialMeshOrientation(Mesh Axis To Orient=<0 -1 0>, Mesh Orientation Mode=::NewEnumerator2, Use Rotation Vector=false)
   /SetVariables(Particles.ScribbleControl:Vector4f=<0 0.5 0.5 1.5>) {isSetParameters=true}
   /SetVariables(
    Particles.GlowColor=&HueShiftLinearColor(Hue=-0.05, Saturation=0.6, Value=0.2, Linear Color=$Particles.Color)
    Particles.GlowOffset=-10
   ) {enabled=false, isSetParameters=true}
  ]
  particleUpdateScript [
   &ParticleState()
   &ScaleMeshSize(Scale Factor=&VectorFromFloat(Value=&FloatFromCurve(FloatCurve=DataInterface() {propertyValues {Curve {keys [
          {arriveTangent=5.0113, interpMode=RCIM_Cubic, leaveTangent=5.0113, tangentMode=RCTM_User, time=0, value=0}
          {interpMode=RCIM_Cubic, tangentMode=RCTM_User, time=0.3, value=1}
          {arriveTangentWeight=0.2325, interpMode=RCIM_Cubic, leaveTangentWeight=0.1744, tangentMode=RCTM_User, time=0.7, value=1}
          {arriveTangent=-2.8957, tangentMode=RCTM_User, time=1, value=0}
         ]}}})))
   &ScaleSpriteSize(
    Scale Sprite Size Mode=::NewEnumerator0
    Uniform Scale Factor=&Multiply_Float(A=&MakeFloatFromVector(Vector=$Particles.Scale, Channel=::NewEnumerator0), B=1.25)
   )
   /SetVariables(Transient.SpinPhase=0.75) {isSetParameters=true}
   /SetVariables(Transient.CenterSpin:NiagaraFloat=&Divide_Float(A=0.25, B=$Transient.SpinPhase)) {isSetParameters=true}
   /SetVariables(Particles.Bob:NiagaraFloat=&Waveform(
     Global Phase=&Add_Float(A=$Particles.Age, B=$Transient.CenterSpin)
     "Waveform [1]"=::NewEnumerator0
     "[1] Phase Scale"=$Transient.SpinPhase
     "[1] Frequency"=1
     "[1] Amplitude Min/Max"=<-1 1>
    )) {isSetParameters=true}
   /SetVariables(Particles.Position=&AddVectorToPosition(
     Position=&VectorFromCurve(VectorCurve=DataInterface() {propertyValues {
        XCurve {keys [
          {arriveTangent=53.7988, interpMode=RCIM_Cubic, leaveTangent=53.7988, leaveTangentWeight=153.2665, tangentMode=RCTM_User, time=0, value=-74}
          {arriveTangent=18.8279, arriveTangentWeight=170.1042, interpMode=RCIM_Cubic, leaveTangent=18.8279, tangentMode=RCTM_User, time=1, value=86}
         ]}
        YCurve {keys [
          {arriveTangent=-29.4987, interpMode=RCIM_Cubic, leaveTangent=-29.4987, tangentMode=RCTM_User, time=0, value=0}
          <at 0.5 -10 RCIM_Cubic>
          {arriveTangent=30.6894, interpMode=RCIM_Cubic, leaveTangent=30.6894, tangentMode=RCTM_User, time=1, value=0}
         ]}
        ZCurve {keys [
          {arriveTangent=59.6951, interpMode=RCIM_Cubic, leaveTangent=59.6951, tangentMode=RCTM_User, time=0, value=1}
          {arriveTangent=-1, interpMode=RCIM_Cubic, leaveTangent=-1, time=0.5, value=20}
          {arriveTangent=-59.5441, interpMode=RCIM_Cubic, leaveTangent=-59.5441, tangentMode=RCTM_User, time=1, value=0}
         ]}
       }})
     Vector=&MakeVector(X=0, Y=0, Z=&Multiply_Float(A=10, B=$Particles.Bob))
    )) {isSetParameters=true, module=SetVariables_#2}
   &UpdateMeshOrientation(Rotation Rate=&Multiply_Float(A=$Particles.Bob, B=0.3))
  ]
  rendererClasses [&NiagaraMeshRendererProperties, &NiagaraSpriteRendererProperties]
  renderers [
   {rendererClass=&NiagaraMeshRendererProperties, rendererIndex=0}
   {rendererClass=&NiagaraMeshRendererProperties, rendererIndex=1}
   {rendererClass=&NiagaraSpriteRendererProperties, rendererIndex=2}
  ]
 }
 Star {
  emitterUpdateScript [
   &EmitterState()
   &SpawnBurst_Instantaneous(Spawn Group=2, Spawn Time=0.1, Use Spawn Probability=false, UseLoopCountLimit=true)
   &SpawnBurst_Instantaneous(Spawn Count=2, Spawn Group=1, Spawn Time=0.2, Use Spawn Probability=false, UseLoopCountLimit=true) {module=SpawnBurst_Instantaneous001}
   &SpawnBurst_Instantaneous(Spawn Time=0.4, Use Spawn Probability=false, UseLoopCountLimit=true) {module=SpawnBurst_Instantaneous002}
   &SpawnBurst_Instantaneous(Spawn Time=0.45, Use Spawn Probability=false, UseLoopCountLimit=true) {module=SpawnBurst_Instantaneous003}
   &SpawnBurst_Instantaneous(Spawn Group=3, Spawn Time=0.55, Use Spawn Probability=false, UseLoopCountLimit=true) {module=SpawnBurst_Instantaneous004}
  ]
  particleSpawnScript [
   &PartitionParticles()
   &InitializeParticle(
    Color=&SelectLinearColorFromArray(
     Array Sampling Mode=::NewEnumerator0
     Direct Array Index=?
     Color Selection Array=DataInterface() {propertyValues {ColorData [<rgba 1 0.2706 0 1>, <rgba 1 0.4167 0 1>, <rgba 1 0.6868 0 1>]}}
     Random Seed=0
    )
    Lifetime=5
    Mesh Scale=&self:ShareYandZ(
     X Scale:NiagaraFloat=1
     Y and Z Scale:NiagaraFloat=&Lerp_Float(
      A=0.5
      B=0.75
      Alpha=&FloatFromCurve(
       FloatCurve=DataInterface() {propertyValues {Curve {keys [
           <at 0 0 RCIM_Cubic>
           {
            arriveTangent=2.6185
            arriveTangentWeight=0.2357
            interpMode=RCIM_Cubic
            leaveTangent=2.6185
            leaveTangentWeight=0.2357
            tangentMode=RCTM_User
            time=0.5
            value=0.5
           }
           {arriveTangentWeight=0.1667, interpMode=RCIM_Cubic, tangentMode=RCTM_User, time=1, value=1}
          ]}}}
       CurveIndex=&RandomRangeFloat(Minimum=0, Maximum=1)
      )
     )
    )
    Mesh Scale Mode=::NewEnumerator4
    Sprite Size Mode=::NewEnumerator3
    UsePositionOffset=false
    Write Material Random=true
   )
   &InitialMeshOrientation(
    Mesh Orientation Mode=::NewEnumerator7
    Rotation=&RandomRangeVector(Minimum=<-1 0 0>, Maximum=<1 0 0>)
    Use Rotation Vector=true
   ) {module=InitialMeshOrientation001}
   &InitialMeshOrientation(
    Mesh Orientation Mode=::NewEnumerator7
    Rotation=&RandomRangeVector(Minimum=<0 0 -1>, Maximum=<0 0 1>)
    Use Rotation Vector=true
   )
   &DynamicMaterialParameters(
    Index 0 Vector 4:Vector4f=&SelectVector4_FromArray(
     Array Sampling Mode:ENiagaraArraySamplingMode=::NewEnumerator0
     Direct Array Index:NiagaraInt32=?
     Interpolation:NiagaraFloat=?
     Direct Array Mode:ENiagara_WrapClamp=::NewEnumerator0
     Evaluation Type:ENiagaraRandomnessEvaluation=::NewEnumerator0
     Normalized Interpolation:NiagaraBool=true
     Recalculate Random Each Loop:NiagaraBool=false
     Randomness Mode:ENiagaraRandomnessMode=::NewEnumerator0
     Random Seed:NiagaraInt32=0
     Fixed Random Seed:NiagaraBool=false
     Vector4 Selection Array:NiagaraDataInterfaceArrayFloat4=DataInterface() {propertyValues {GpuSyncMode=SyncCpuToGpu, MaxElements=0, FloatData [<0.7 0.5 0.04 0>, <0.73 0 -0.03 -0.06>, <0.8 -0.3 0 0>, <0.85 -0.1 0.01 0.02>]}}
    )
    Index 1 Param 1=&RandomRangeFloat(Minimum=0, Maximum=1)
    Index 1 Param 2=&RandomRangeFloat(Minimum=0, Maximum=1)
    Index 1 Param 3=&FloatFromCurve(
     FloatCurve=DataInterface() {propertyValues {Curve {keys [{interpMode=RCIM_Cubic, leaveTangentWeight=0.3973, tangentMode=RCTM_User, time=0, value=0.5}, <at 1 1 RCIM_Cubic>]}}}
     CurveIndex=&RandomRangeFloat(Minimum=0, Maximum=1)
    )
    Index 1 Param 4=1
    Module.Index1Float0Write=true
    Module.Index1Float1Write=true
    Module.Index1Float2Write=true
    Module.Index1Float3Write=true
    Write Parameter Index 0=::NewEnumerator4
    Write Parameter Index 1=::NewEnumerator1
   )
   /SetVariables(Particles.MeshIndex=&RandomRangeInt(Minimum=0, Maximum=1)) {isSetParameters=true, module=SetVariables_#4}
   /SetVariables(Particles.StaticScale:NiagaraFloat=&FloatFromCurve(FloatCurve=DataInterface() {propertyValues {Curve {keys [<at 0 0>,<at 0.55 1>]}}}, CurveIndex=$Emitter.Age)) {isSetParameters=true, module=SetVariables_#5}
   /SetVariables() {isSetParameters=true, module=SetVariables_#6}
   &ShapeLocation(
    Apply To Particle Position=::NewEnumerator1
    Box Size=<30 10 100>
    MaskSpawnGroup=false
    Offset=<-35 -50 -13>
    Offset Mode=::NewEnumerator0
    Override Seed=false
    Shape Primitive=::NewEnumerator2
    Surface Only=false
    UseOwnerScale=true
   )
   &ShapeLocation(
    Apply To Particle Position=::NewEnumerator1
    Box Size=<60 10 70>
    MaskSpawnGroup=false
    Offset=<40 -50 0>
    Offset Mode=::NewEnumerator0
    Override Seed=false
    Shape Primitive=::NewEnumerator2
    Surface Only=false
    UseOwnerScale=true
   ) {module=ShapeLocation001}
   &ShapeLocation(
    Apply To Particle Position=::NewEnumerator1
    Box Size=<40 40 40>
    MaskSpawnGroup=false
    Offset=<60 -50 -50>
    Offset Mode=::NewEnumerator0
    Override Seed=false
    Shape Primitive=::NewEnumerator2
    Surface Only=false
    UseOwnerScale=true
   ) {module=ShapeLocation002}
   &self:DistributeLocations(Placement Region:NiagaraInt32=&MaskIntBySpawnGroup(
     Masked Int=&RandomRangeInt(Minimum=0, Maximum=1)
     Passthrough Int=&SpawnGroup()
     Spawn Group Mask=3
    )) {module=DistributeLocations}
  ]
  particleUpdateScript [
   &ParticleState(Kill Particles When Lifetime Has Elapsed=false, Let Infinitely Lived Particles Die When Emitter Deactivates=true)
   /SetVariables(
    Transient.AgeAdjusted:NiagaraFloat=&Multiply_Float(A=$Particles.Age, B=&Lerp_Float(A=1.5, B=2, Alpha=$Particles.StaticScale))
    Transient.MotionAdjust:NiagaraFloat=&Lerp_Float(A=0.66, B=0.5, Alpha=$Particles.StaticScale)
   ) {isSetParameters=true, module=SetVariables_#7}
   /SetVariables(
    Particles.BouncePosition:NiagaraPosition=&MakeVector(
     X=&FloatFromCurve(
      FloatCurve=$System.BounceOffset
      CurveIndex=$Transient.AgeAdjusted
      Scale Curve=&Multiply_Float(
       A=$Transient.MotionAdjust
       B=&UniformAOrBFloat(A=-1, Evaluation Type=::NewEnumerator0, Recalculate Random Each Loop=true, Override Seed=false)
      )
     )
     Y=0
     Z=&FloatFromCurve(FloatCurve=$System.BouncePosition, CurveIndex=$Transient.AgeAdjusted)
    )
    Transient.BouncePhase:NiagaraFloat=&FloatFromCurve(FloatCurve=$System.BouncePhase, CurveIndex=$Transient.AgeAdjusted, Scale Curve=$Transient.MotionAdjust)
   ) {isSetParameters=true, module=SetVariables_#3}
   &UpdateMeshOrientation(
    Rotation Coordinate Space=::NewEnumerator1
    Rotation Rate=&Waveform(
     Global Phase=$Particles.Age
     Global Amplitude Scale=&Multiply_Float(
      A=&UniformAOrBFloat(A=-1, Evaluation Type=::NewEnumerator0, Recalculate Random Each Loop=true, Override Seed=false)
      B=&RandomRangeFloat(Minimum=0.75, Maximum=1.25)
     )
     "Waveform [1]"=::NewEnumerator0
     "[1] Phase Scale"=0.1
     "[1] Frequency"=1
     "[1] Amplitude Min/Max"=<-1 1>
     "[1] Scale"=&LerpMultipleFloats(
      Float Count:ENiagara_MultipleLerpCount=::NewEnumerator1
      1:NiagaraFloat=-5
      2:NiagaraFloat=1
      3:NiagaraFloat=0.1
      4:NiagaraFloat=0.03
      5:NiagaraFloat=?
      Alpha:NiagaraFloat=&Multiply_Float(A=$Transient.BouncePhase, B=0.25)
     )
    )
    Rotation Vector=<0 0 1>
   )
   /SetVariables(Particles.SpinPhase:NiagaraFloat=&RandomRangeFloat(Minimum=0.05, Maximum=0.12)) {isSetParameters=true, module=SetVariables_#8}
   /SetVariables(Transient.CenterSpin:NiagaraFloat=&Divide_Float(A=0.25, B=$Particles.SpinPhase)) {isSetParameters=true, module=SetVariables_#9}
   &UpdateMeshOrientation(Rotation Rate=&Waveform(
     Global Phase=&Add_Float(A=$Particles.Age, B=$Transient.CenterSpin)
     Global Amplitude Scale=&RandomRangeFloat(Minimum=-0.08, Maximum=0.08)
     "Waveform [1]"=::NewEnumerator0
     "[1] Phase Scale"=$Particles.SpinPhase
     "[1] Frequency"=1
     "[1] Amplitude Min/Max"=<-1 1>
    )) {module=UpdateMeshOrientation001}
   /SetVariables(Particles.OverallPosition:NiagaraPosition=&AddVectorToPosition(
     Position=$Particles.Initial.Position
     Vector=&Add_Vector(
      A=$Particles.RandomOffset
      B=&Multiply_VectorByFloat(
       Vector=Linked() {linkedVariable {name=Particles.BouncePosition, type=NiagaraPosition}}
       Float=&RandomRangeFloat(Minimum=0.5, Maximum=1)
      )
     )
    )) {isSetParameters=true, module=SetVariables_#10}
   /SetVariables(Particles.RandomOffset:Vector3f=&self:ClampVectorByLength(
     Vector:Vector3f=/Niagara/DynamicInputs/Lerp/SmoothLerpOverTime_Vector(
      Convergence Rate:NiagaraFloat=0.5
      Smooth Value:Vector3f=$Particles.RandomOffset
      Target Value:Vector3f=&Add_Vector(
       A=$Particles.RandomOffset
       B=&RandomVector(
        Vector Scale=&MakeCustomFloatFromBool(Bool=$Output.ParticleState.FirstFrame, True Float=0, False Float=5)
        Evaluation Type=::NewEnumerator1
        Randomness Mode=::NewEnumerator1
        Random Seed=&MakeIntFromFloat(Float=&Add_Float(A=$Particles.Age, B=&RandomRangeFloat(Minimum=0, Maximum=10)))
        Fixed Random Seed=true
       )
      )
     )
     Max Length:NiagaraFloat=20
     Min Length:NiagaraFloat=0
    )) {isSetParameters=true, module=SetVariables_#11}
  ]
  rendererClasses [&NiagaraMeshRendererProperties, &NiagaraSpriteRendererProperties]
  renderers [
   {rendererClass=&NiagaraMeshRendererProperties, rendererIndex=0}
   {rendererClass=&NiagaraSpriteRendererProperties, rendererIndex=1}
  ]
 }
 Star001 {
  bEnabled=false
  emitterUpdateScript [
   &EmitterState()
   &SpawnBurst_Instantaneous(Spawn Group=2, Spawn Time=0.1, Use Spawn Probability=false, UseLoopCountLimit=true)
   &SpawnBurst_Instantaneous(Spawn Count=2, Spawn Group=1, Spawn Time=0.2, Use Spawn Probability=false, UseLoopCountLimit=true) {module=SpawnBurst_Instantaneous001}
   &SpawnBurst_Instantaneous(Spawn Time=0.4, Use Spawn Probability=false, UseLoopCountLimit=true) {module=SpawnBurst_Instantaneous002}
   &SpawnBurst_Instantaneous(Spawn Time=0.45, Use Spawn Probability=false, UseLoopCountLimit=true) {module=SpawnBurst_Instantaneous003}
   &SpawnBurst_Instantaneous(Spawn Group=3, Spawn Time=0.55, Use Spawn Probability=false, UseLoopCountLimit=true) {module=SpawnBurst_Instantaneous004}
  ]
  particleSpawnScript [
   &InitializeParticle(
    Color=&SelectLinearColorFromArray(
     Array Sampling Mode=::NewEnumerator0
     Direct Array Index=?
     Color Selection Array=DataInterface() {propertyValues {ColorData [<rgba 1 0.2706 0 1>, <rgba 1 0.4167 0 1>, <rgba 1 0.6868 0 1>]}}
     Random Seed=0
    )
    Lifetime=5
    Mesh Scale=&self:ShareYandZ(
     X Scale:NiagaraFloat=1
     Y and Z Scale:NiagaraFloat=&Lerp_Float(
      A=0.5
      B=0.75
      Alpha=&FloatFromCurve(
       FloatCurve=DataInterface() {propertyValues {Curve {keys [
           <at 0 0 RCIM_Cubic>
           {
            arriveTangent=2.6185
            arriveTangentWeight=0.2357
            interpMode=RCIM_Cubic
            leaveTangent=2.6185
            leaveTangentWeight=0.2357
            tangentMode=RCTM_User
            time=0.5
            value=0.5
           }
           {arriveTangentWeight=0.1667, interpMode=RCIM_Cubic, tangentMode=RCTM_User, time=1, value=1}
          ]}}}
       CurveIndex=&RandomRangeFloat(Minimum=0, Maximum=1)
      )
     )
    )
    Mesh Scale Mode=::NewEnumerator4
    Sprite Size Mode=::NewEnumerator3
    UsePositionOffset=false
    Write Material Random=true
   )
   &InitialMeshOrientation(
    Mesh Orientation Mode=::NewEnumerator7
    Rotation=&RandomRangeVector(Minimum=<-1 0 0>, Maximum=<1 0 0>)
    Use Rotation Vector=true
   ) {module=InitialMeshOrientation001}
   &InitialMeshOrientation(
    Mesh Orientation Mode=::NewEnumerator7
    Rotation=&RandomRangeVector(Minimum=<0 0 -1>, Maximum=<0 0 1>)
    Use Rotation Vector=true
   )
   &DynamicMaterialParameters(
    Index 0 Vector 4:Vector4f=&SelectVector4_FromArray(
     Array Sampling Mode:ENiagaraArraySamplingMode=::NewEnumerator0
     Direct Array Index:NiagaraInt32=?
     Interpolation:NiagaraFloat=?
     Direct Array Mode:ENiagara_WrapClamp=::NewEnumerator0
     Evaluation Type:ENiagaraRandomnessEvaluation=::NewEnumerator0
     Normalized Interpolation:NiagaraBool=true
     Recalculate Random Each Loop:NiagaraBool=false
     Randomness Mode:ENiagaraRandomnessMode=::NewEnumerator0
     Random Seed:NiagaraInt32=0
     Fixed Random Seed:NiagaraBool=false
     Vector4 Selection Array:NiagaraDataInterfaceArrayFloat4=DataInterface() {propertyValues {GpuSyncMode=SyncCpuToGpu, MaxElements=0, FloatData [<0.7 0.5 0.04 0>, <0.73 0 -0.03 -0.06>, <0.8 -0.3 0 0>, <0.85 -0.1 0.01 0.02>]}}
    )
    Index 1 Param 1=&RandomRangeFloat(Minimum=0, Maximum=1)
    Index 1 Param 2=&RandomRangeFloat(Minimum=0, Maximum=1)
    Index 1 Param 3=&FloatFromCurve(
     FloatCurve=DataInterface() {propertyValues {Curve {keys [{interpMode=RCIM_Cubic, leaveTangentWeight=0.3973, tangentMode=RCTM_User, time=0, value=0.5}, <at 1 1 RCIM_Cubic>]}}}
     CurveIndex=&RandomRangeFloat(Minimum=0, Maximum=1)
    )
    Index 1 Param 4=1
    Module.Index1Float0Write=true
    Module.Index1Float1Write=true
    Module.Index1Float2Write=true
    Module.Index1Float3Write=true
    Write Parameter Index 0=::NewEnumerator4
    Write Parameter Index 1=::NewEnumerator1
   )
   /SetVariables(Particles.MeshIndex=&RandomRangeInt(Minimum=0, Maximum=1)) {isSetParameters=true, module=SetVariables_#4}
   /SetVariables(Particles.StaticScale:NiagaraFloat=&FloatFromCurve(FloatCurve=DataInterface() {propertyValues {Curve {keys [<at 0 0>,<at 0.55 1>]}}}, CurveIndex=$Emitter.Age)) {isSetParameters=true, module=SetVariables_#5}
   /SetVariables() {isSetParameters=true, module=SetVariables_#6}
   &ShapeLocation(
    Apply To Particle Position=::NewEnumerator1
    Box Size=<30 40 100>
    MaskSpawnGroup=false
    Offset=<-35 -50 -13>
    Offset Mode=::NewEnumerator0
    Override Seed=false
    Shape Primitive=::NewEnumerator2
    Surface Only=false
    UseOwnerScale=true
   )
   &ShapeLocation(
    Apply To Particle Position=::NewEnumerator1
    Box Size=<60 40 70>
    MaskSpawnGroup=false
    Offset=<40 -50 0>
    Offset Mode=::NewEnumerator0
    Override Seed=false
    Shape Primitive=::NewEnumerator2
    Surface Only=false
    UseOwnerScale=true
   ) {module=ShapeLocation001}
   &ShapeLocation(
    Apply To Particle Position=::NewEnumerator1
    Box Size=<40 40 40>
    MaskSpawnGroup=false
    Offset=<60 -50 -50>
    Offset Mode=::NewEnumerator0
    Override Seed=false
    Shape Primitive=::NewEnumerator2
    Surface Only=false
    UseOwnerScale=true
   ) {module=ShapeLocation002}
   &self:DistributeLocations(Placement Region:NiagaraInt32=&MaskIntBySpawnGroup(
     Masked Int=&RandomRangeInt(Minimum=0, Maximum=1)
     Passthrough Int=&SpawnGroup()
     Spawn Group Mask=3
    )) {module=DistributeLocations}
  ]
  particleUpdateScript [
   &ParticleState(Kill Particles When Lifetime Has Elapsed=false, Let Infinitely Lived Particles Die When Emitter Deactivates=true)
   /SetVariables(
    Transient.AgeAdjusted:NiagaraFloat=&Multiply_Float(A=$Particles.Age, B=&Lerp_Float(A=1.5, B=2, Alpha=$Particles.StaticScale))
    Transient.MotionAdjust:NiagaraFloat=&Lerp_Float(A=0.66, B=0.5, Alpha=$Particles.StaticScale)
   ) {isSetParameters=true, module=SetVariables_#7}
   /SetVariables(
    Particles.BouncePosition:NiagaraPosition=&MakeVector(
     X=&FloatFromCurve(
      FloatCurve=$System.BounceOffset
      CurveIndex=$Transient.AgeAdjusted
      Scale Curve=&Multiply_Float(
       A=$Transient.MotionAdjust
       B=&UniformAOrBFloat(A=-1, Evaluation Type=::NewEnumerator0, Recalculate Random Each Loop=true, Override Seed=false)
      )
     )
     Y=0
     Z=&FloatFromCurve(FloatCurve=$System.BouncePosition, CurveIndex=$Transient.AgeAdjusted)
    )
    Transient.BouncePhase:NiagaraFloat=&FloatFromCurve(FloatCurve=$System.BouncePhase, CurveIndex=$Transient.AgeAdjusted, Scale Curve=$Transient.MotionAdjust)
   ) {isSetParameters=true, module=SetVariables_#3}
   &UpdateMeshOrientation(
    Rotation Coordinate Space=::NewEnumerator1
    Rotation Rate=&Waveform(
     Global Phase=$Particles.Age
     Global Amplitude Scale=&Multiply_Float(
      A=&UniformAOrBFloat(A=-1, Evaluation Type=::NewEnumerator0, Recalculate Random Each Loop=true, Override Seed=false)
      B=&RandomRangeFloat(Minimum=0.75, Maximum=1.25)
     )
     "Waveform [1]"=::NewEnumerator0
     "[1] Phase Scale"=0.1
     "[1] Frequency"=1
     "[1] Amplitude Min/Max"=<-1 1>
     "[1] Scale"=&LerpMultipleFloats(
      Float Count:ENiagara_MultipleLerpCount=::NewEnumerator1
      1:NiagaraFloat=-5
      2:NiagaraFloat=1
      3:NiagaraFloat=0.1
      4:NiagaraFloat=0.03
      5:NiagaraFloat=?
      Alpha:NiagaraFloat=&Multiply_Float(A=$Transient.BouncePhase, B=0.25)
     )
    )
    Rotation Vector=<0 0 1>
   )
   /SetVariables(Particles.SpinPhase:NiagaraFloat=&RandomRangeFloat(Minimum=0.05, Maximum=0.12)) {isSetParameters=true, module=SetVariables_#8}
   /SetVariables(Transient.CenterSpin:NiagaraFloat=&Divide_Float(A=0.25, B=$Particles.SpinPhase)) {isSetParameters=true, module=SetVariables_#9}
   &UpdateMeshOrientation(Rotation Rate=&Waveform(
     Global Phase=&Add_Float(A=$Particles.Age, B=$Transient.CenterSpin)
     Global Amplitude Scale=&RandomRangeFloat(Minimum=-0.08, Maximum=0.08)
     "Waveform [1]"=::NewEnumerator0
     "[1] Phase Scale"=$Particles.SpinPhase
     "[1] Frequency"=1
     "[1] Amplitude Min/Max"=<-1 1>
    )) {module=UpdateMeshOrientation001}
   /SetVariables(Particles.OverallPosition:NiagaraPosition=&AddVectorToPosition(
     Position=$Particles.Initial.Position
     Vector=&Add_Vector(
      A=$Particles.RandomOffset
      B=&Multiply_VectorByFloat(
       Vector=Linked() {linkedVariable {name=Particles.BouncePosition, type=NiagaraPosition}}
       Float=&RandomRangeFloat(Minimum=0.5, Maximum=1)
      )
     )
    )) {isSetParameters=true, module=SetVariables_#10}
   /SetVariables(Particles.RandomOffset:Vector3f=&self:ClampVectorByLength(
     Vector:Vector3f=&SmoothLerpOverTime_Vector(
      Convergence Rate:NiagaraFloat=0.5
      Smooth Value:Vector3f=$Particles.RandomOffset
      Target Value:Vector3f=&Add_Vector(
       A=$Particles.RandomOffset
       B=&RandomVector(
        Vector Scale=&MakeCustomFloatFromBool(Bool=$Output.ParticleState.FirstFrame, True Float=0, False Float=5)
        Evaluation Type=::NewEnumerator1
        Randomness Mode=::NewEnumerator1
        Random Seed=&MakeIntFromFloat(Float=&Add_Float(A=$Particles.Age, B=&RandomRangeFloat(Minimum=0, Maximum=10)))
        Fixed Random Seed=true
       )
      )
     )
     Max Length:NiagaraFloat=20
     Min Length:NiagaraFloat=0
    )) {isSetParameters=true, module=SetVariables_#11}
  ]
  rendererClasses [&NiagaraMeshRendererProperties, &NiagaraSpriteRendererProperties]
  renderers [
   {rendererClass=&NiagaraMeshRendererProperties, rendererIndex=0}
   {rendererClass=&NiagaraSpriteRendererProperties, rendererIndex=1}
  ]
 }
 StarsPoints {
  emitterUpdateScript [
   &EmitterState()
   &SpawnBurst_Instantaneous(Spawn Count=4, Spawn Time=0.3, Use Spawn Probability=false, UseLoopCountLimit=true)
   &SpawnRate(
    SpawnRate=&FloatFromCurve(
     FloatCurve=DataInterface() {propertyValues {Curve {keys [{tangentMode=RCTM_User, time=0.25, value=0}, {tangentMode=RCTM_User, time=1, value=1}]}}}
     CurveIndex=$Emitter.Age
     Scale Curve=4
    )
    Use Spawn Probability=false
   )
  ]
  particleSpawnScript [
   &InitializeParticle(
    Color=<rgba 1 0.8299 0.4168 1>
    Lifetime Max=5
    Lifetime Mode=::NewEnumerator1
    Sprite Rotation Mode=::NewEnumerator2
    Sprite Size Mode=::NewEnumerator1
    Uniform Sprite Size Max=17
    Uniform Sprite Size Min=10
    UsePositionOffset=false
    Write Material Random=true
   )
   &ShapeLocation(
    MaskSpawnGroup=false
    Non Uniform Scale=<1.5 0.75 1.3>
    Offset=<13 -50 -20>
    Offset Mode=::NewEnumerator0
    Override Seed=false
    Sphere Radius=&RandomRangeFloat(Minimum=25, Maximum=60)
    UseOwnerScale=true
   )
  ]
  particleUpdateScript [
   &ParticleState()
   &SpriteRotationRate(Rotation Rate=&RandomRangeFloat(Minimum=-30, Maximum=30))
   &ScaleColor(
    Scale Alpha=&FNTimeOfDayScalar(Night Scale=1, Morning Scale=0.3, Evening Scale=0.3)
    Scale RGB=&VectorFromFloat(Value=&FloatFromCurve(FloatCurve=DataInterface() {propertyValues {Curve {keys [
          {interpMode=RCIM_Cubic, tangentMode=RCTM_User, time=0.3, value=1}
          {arriveTangent=-2.2588, interpMode=RCIM_Cubic, leaveTangent=-2.2588, tangentMode=RCTM_User, time=1, value=0.25}
         ]}}}))
    ScaleA=true
    ScaleRGB=true
   )
   &ScaleSpriteSize(Scale Factor=&Vector2DFromFloat(Value=&FloatFromCurve(FloatCurve=DataInterface() {propertyValues {Curve {keys [
          {interpMode=RCIM_Cubic, tangentMode=RCTM_User, time=0, value=0}
          {arriveTangentWeight=0.1863, interpMode=RCIM_Cubic, leaveTangentWeight=0.1863, tangentMode=RCTM_User, time=0.5, value=1}
          {arriveTangent=-4.3398, tangentMode=RCTM_User, time=1, value=0}
         ]}}})))
   &ScaleColor(
    Scale Alpha=&Multiply_Float(
     A=&FNTimeOfDayScalar(Night Scale=1, Morning Scale=0.3, Evening Scale=0.3)
     B=&FloatFromCurve(FloatCurve=DataInterface() {propertyValues {Curve {keys [
          {interpMode=RCIM_Cubic, tangentMode=RCTM_User, time=0.3, value=1}
          {arriveTangent=-2.2588, interpMode=RCIM_Cubic, leaveTangent=-2.2588, tangentMode=RCTM_User, time=1, value=0}
         ]}}})
    )
    ScaleA=true
    ScaleRGB=false
   ) {enabled=false, module=ScaleColor001}
   &ScaleSpriteSize(Scale Factor=&Vector2DFromFloat(Value=&FloatFromCurve(FloatCurve=DataInterface() {propertyValues {Curve {keys [
          {interpMode=RCIM_Cubic, tangentMode=RCTM_User, time=0, value=0}
          {arriveTangentWeight=0.1863, interpMode=RCIM_Cubic, leaveTangentWeight=0.1863, tangentMode=RCTM_User, time=0.5, value=1}
          {arriveTangent=-1.3908, tangentMode=RCTM_User, time=1, value=0.75}
         ]}}}))) {enabled=false, module=ScaleSpriteSize001}
  ]
  rendererClasses [&NiagaraSpriteRendererProperties]
  renderers [{rendererClass=&NiagaraSpriteRendererProperties}]
 }
}
summary {
 systemName=NS_Emote_Guitar_StarGazer_Celestial
 emitters [
  {emitterName=Moon, rendererClasses [&NiagaraSpriteRendererProperties, &NiagaraMeshRendererProperties]}
  {emitterName=Star, rendererClasses [&NiagaraMeshRendererProperties, &NiagaraSpriteRendererProperties]}
  {bEnabled=false, emitterName=Minimal, rendererClasses [&NiagaraSpriteRendererProperties]}
  {bEnabled=false, emitterName=Minimal001, rendererClasses [&NiagaraSpriteRendererProperties]}
  {bEnabled=false, emitterName=Minimal002, rendererClasses [&NiagaraSpriteRendererProperties]}
  {bEnabled=false, emitterName=Minimal003, rendererClasses [&NiagaraSpriteRendererProperties]}
  {emitterName=ShootingStar, rendererClasses [&NiagaraMeshRendererProperties, &NiagaraSpriteRendererProperties]}
  {emitterName=StarsPoints, rendererClasses [&NiagaraSpriteRendererProperties]}
  {emitterName=Clouds, rendererClasses [&NiagaraMeshRendererProperties]}
  {bEnabled=false, emitterName=Minimal004, rendererClasses [&NiagaraSpriteRendererProperties]}
  {
   bEnabled=false
   emitterName=Star001
   rendererClasses [&NiagaraMeshRendererProperties, &NiagaraSpriteRendererProperties]
  }
 ]
 userVariables []
}
